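# Material function for Military Base props that spins fan meshes through vertex animation, gated by the UseFanRotation static switch.
# Periodic time drives RotateAboutAxis around a local-space axis parameter, pivoting on the object pivot plus PivotPointBias, to produce World Position Offset.
# It also rotates the vertex normal, combines it with the incoming normal and converts to tangent space.
# An actor-position phase offset is computed but left unconnected.
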MaterialFunction /Game/Creative/Sets/MilitaryBase/Props/MF_Military_FanSpin
graph {nodes [
  n1 = Multiply(A=Time() {bOverride_Period=true, period=5}, B=Divide() {constB=5})
  n2 = ActorPositionWS()
  n3 = Constant() {r=1}
  Add(A=@n1, B=Frac(None=DotProduct(A=@n2, B=@n3)))
  n4 = FunctionInput() {inputName=Material Attributes, inputType=FunctionInput_MaterialAttributes}
  n5 = GetMaterialAttributes(None=@n4) {attributeGetTypes [#1]}
  n6 = Normalize(VectorInput=Transform(None=VectorParameter() {group=FanRotation, parameterName=NormalizedRotationAxis, defaultValue=<rgba 1 0 0 0>}.RGB) {transformSourceType=TRANSFORMSOURCE_Local})
  n7 = RotateAboutAxis(
   NormalizedRotationAxis=@n6
   RotationAngle=@n1
   PivotPoint=Add(
    A=MaterialFunctionCall() {materialFunction=&ObjectPivotPoint}.Object Pivot Location
    B=VectorParameter() {group=FanRotation, parameterName=PivotPointBias}.RGB
   )
   Position=WorldPosition().XYZ
  )
  n8 = RotateAboutAxis(NormalizedRotationAxis=@n6, RotationAngle=@n1, PivotPoint=Constant(), Position=VertexNormalWS())
  n9 = Transform(None=Normalize(VectorInput=Add(A=@n5.Normal, B=@n8))) {transformSourceType=TRANSFORMSOURCE_World, transformType=TRANSFORM_Tangent}
  FunctionOutput(None=StaticSwitchParameter(
    True=SetMaterialAttributes!(MaterialAttributes=@n5.MaterialAttributes, WorldPositionOffset=@n7, Normal=@n9) {attributeSetTypes [#2, #1]}
    False=@n4
   ) {group=FanRotation, parameterName=UseFanRotation})
 ]}
