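; Child Blueprint of a military base ceiling light prop, adding a second flare and light cone (FlareAlt, LightConeAlt) to the inherited light visuals.
; Overrides call the parent, then apply the same client visibility, flare size index and loaded materials to the alternate components.
; GetOverrideMeshMaterial returns the current flare or cone material for those meshes, otherwise deferring to the parent.
; The construction script gives FlareAlt absolute rotation and copies the primary Flare's relative scale.
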(event EventUpdateLightVisualsOnClient
  (|CallParentFunction)
  (Rendering|SetVisibility (Variables|Default|GetLightConeAlt) (Math|Conversions|ToBoolean(Integer) (Variables|Default|GetIsVisibileIndex)) false)
  (CallFunction|SetFlareVisibility (Variables|Default|GetFlareAlt) (Variables|Default|GetFlareSizeIndex)))

(event EventFlareMaterialLoaded
  (params (LoadedObject :type Object%20Reference))
  (|CallParentFunction LoadedObject)
  (bind _currentflarematerial (Variables|Default|GetCurrentFlareMaterial)
    (:then
      (Rendering|Material|SetMaterial (Variables|Default|GetFlareAlt) 0 _currentflarematerial))
    (:else)))

(event EventLightConeMaterialLoaded
  (params (LoadedObject :type Object%20Reference))
  (|CallParentFunction LoadedObject)
  (bind _currentlightconematerial (Variables|Default|GetCurrentLightConeMaterial)
    (:then
      (Rendering|Material|SetMaterial (Variables|Default|GetLightConeAlt) 0 _currentlightconematerial))
    (:else)))

(fn GetOverrideMeshMaterial
  (params (MeshComp :type Static%20Mesh%20Component%20Object%20Reference) (MatIdx :type Integer))
  (Utilities|FlowControl|Sequence
    (:then_0
      (if (Utilities|Equal(Object) (Variables|GetMeshComp) (Variables|Default|GetFlareAlt))
        (bind _currentflarematerial (Variables|Default|GetCurrentFlareMaterial)
          (:then
            (return _currentflarematerial))
          (:else))))
    (:then_1
      (if (Utilities|Equal(Object) (Variables|GetMeshComp) (Variables|Default|GetLightConeAlt))
        (bind _currentlightconematerial (Variables|Default|GetCurrentLightConeMaterial)
          (:then
            (return _currentlightconematerial))
          (:else))))
    (:then_2
      (return (|CallParentFunction (Variables|GetMeshComp) (Variables|GetMatIdx))))))

(fn ConstructionScript
  (|CallParentFunction)
  (Variables|/Script/Engine.SceneComponent|SetbAbsoluteRotation true (Variables|Default|GetFlareAlt))
  (Transformation|SetRelativeScale3D (Variables|Default|GetFlareAlt) (Variables|/Script/Engine.SceneComponent|GetRelativeScale3D (Variables|Default|GetFlare))))
