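; Military Base floor light prop for Creative, with a point light whose visibility and team color are driven by index variables.
; On the client visual update, after the parent call, it toggles the light from the visibility index and applies the matching color from CreativeTeamColors class defaults if the color index is valid.
; It overrides the parent's time-of-day defaults only to force the flag disabling TOD lights and emissive values to false.
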(event EventUpdateLightVisualsOnClient
  (bind _teamlightcolors (Class|GetClassDefaults "/Game/Creative/Sets/MilitaryBase/Props/CreativeTeamColors.CreativeTeamColors_C"))
  (bind _lightcolorindex (Variables|Default|GetLightColorIndex))
  (|CallParentFunction)
  (Rendering|SetVisibility (Variables|Default|GetPointLight) (Math|Conversions|ToBoolean(Integer) (Variables|Default|GetIsVisibileIndex)) false)
  (if (Utilities|Array|IsValidIndex _teamlightcolors _lightcolorindex)
    (Rendering|Components|Light|SetLightColor (Variables|Default|GetPointLight) (Utilities|Array|Get(aref) _teamlightcolors _lightcolorindex) true)))

(fn GetTimeOfDayBlueprintDefaultVariables
  (bind _outvariables (|CallParentFunction))
  (bind (_outvariables_alternateshadowstaticmesh _outvariables_volumetriclightscatteringintensity _outvariables_bdisabletodlightsandmaterialemissivevalues _outvariables_bdisablestaticmeshshadowcastingwhenlightsareactive _outvariables_buseanalternateshadowmeshwhenthelightisoff _outvariables_bcastvolumetricshadows) (Utilities|Struct|BreakTimeOfDayBlueprintDefaultVariables _outvariables))
  (return (Utilities|Struct|MakeTimeOfDayBlueprintDefaultVariables _outvariables_alternateshadowstaticmesh _outvariables_volumetriclightscatteringintensity false _outvariables_bdisablestaticmeshshadowcastingwhenlightsareactive _outvariables_buseanalternateshadowmeshwhenthelightisoff _outvariables_bcastvolumetricshadows)))

(fn ConstructionScript
  (|CallParentFunction))
