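; Master Blueprint for Military Base Creative light props, updating spotlight, flare, light cone and mesh visuals from toy options on clients able to run cosmetic events.
; Begin play and option changes or loads set spotlight visibility and team color plus flare visibility and scale.
; Color-indexed soft materials load asynchronously through events child Blueprints can use, with a deactivated mesh material when off.
; Applied materials are cached for GetOverrideMeshMaterial, and construction fixes the flare's world rotation.
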(event EventBeginPlay
  (Networking|CanExecuteCosmeticEvents
    (:True
      (CallFunction|UpdateLightVisualsonClient))))

(event OnAnyPropertyChanged(ToyOptionsComponent)
  (Networking|CanExecuteCosmeticEvents
    (:True
      (CallFunction|UpdateLightVisualsonClient))))

(event OnPropertiesLoaded(ToyOptionsComponent)
  (Networking|CanExecuteCosmeticEvents
    (:True
      (CallFunction|UpdateLightVisualsonClient))))

(event Custom|UpdateLightVisualsOnClient
  (bind _isvisibileindex (Variables|Default|GetIsVisibileIndex))
  (bind _returnvalue (Math|Conversions|ToBoolean(Integer) _isvisibileindex))
  (bind _lightcolorindex (Variables|Default|GetLightColorIndex))
  (bind _flarematerials (Variables|Default|GetFlareMaterials))
  (bind _output (Utilities|Array|Get(acopy) _flarematerials _lightcolorindex))
  (bind _lightcolorindex_2 (Variables|Default|GetLightColorIndex))
  (bind _meshvisiblematerials (Variables|Default|GetMeshVisibleMaterials))
  (bind _output_2 (Utilities|Array|Get(acopy) _meshvisiblematerials _lightcolorindex_2))
  (bind _lightcolorindex_3 (Variables|Default|GetLightColorIndex))
  (bind _lightconematerials (Variables|Default|GetLightConeMaterials))
  (bind _output_3 (Utilities|Array|Get(acopy) _lightconematerials _lightcolorindex_3))
  (bind _teamlightcolors (Class|GetClassDefaults "/Game/Creative/Sets/MilitaryBase/Props/CreativeTeamColors.CreativeTeamColors_C"))
  (bind _meshlightdeactivatedmaterial (Variables|Default|GetMeshLightDeactivatedMaterial))
  (bind _returnvalue_2 (select _returnvalue _output_2 _meshlightdeactivatedmaterial))
  (bind _lightcolorindex_4 (Variables|Default|GetLightColorIndex))
  ;[1 1 1] Update Visuals
  (Rendering|SetVisibility (Variables|Default|GetSpotLight) (Math|Conversions|ToBoolean(Integer) (Variables|Default|GetIsVisibileIndex)) false)
  (CallFunction|SetFlareVisibility (Variables|Default|GetFlare) (Variables|Default|GetFlareSizeIndex))
  (Transformation|SetRelativeScale3D (Variables|Default|GetFlareParent) (Utilities|Select :Index (Variables|Default|GetFlareSizeIndex) :Option%200 "0, 0, 0" :Option%201 "1.000000,1.000000,0.500000" :Option%202 "1.000000,1.000000,1.000000" :Option%203 "1.000000,1.000000,2.250000"))
  (if (Utilities|Array|IsValidIndex _teamlightcolors _lightcolorindex_4)
    (Rendering|Components|Light|SetLightColor (Variables|Default|GetSpotLight) (Utilities|Array|Get(aref) _teamlightcolors _lightcolorindex_4) true)
    (else
      _))
  (if (Utilities|IsValidSoftObjectReference _output)
    (bind _object (Utilities|AsyncLoadAsset _output)
      (:then
        _)
      (:Completed
        (CallFunction|FlareMaterialLoaded _object)))
    (else
      _))
  (if (Utilities|IsValidSoftObjectReference _output_3)
    (bind _object_2 (Utilities|AsyncLoadAsset _output_3)
      (:then
        _)
      (:Completed
        (CallFunction|LightConeMaterialLoaded _object_2)))
    (else
      _))
  (if (Utilities|Array|IsValidIndex _meshvisiblematerials _lightcolorindex_2)
    (if (Utilities|IsValidSoftObjectReference _returnvalue_2)
      (bind _object_3 (Utilities|AsyncLoadAsset _returnvalue_2)
        (:Completed
          (CallFunction|MeshMaterialLoaded _object_3))))))

(event Custom|FlareMaterialLoaded
  (params (LoadedObject :type Object%20Reference))
  ;[1 1 1] Update Visuals
  ;;[1 1 1] CEs are done for child blueprints
  (bind _asmaterial_interface (Utilities|Casting|CastToMaterialInterface LoadedObject))
  (Rendering|Material|SetMaterial (Variables|Default|GetFlare) 0 _asmaterial_interface)
  (Variables|Default|SetCurrentFlareMaterial _asmaterial_interface))

(event Custom|LightConeMaterialLoaded
  (params (LoadedObject :type Object%20Reference))
  ;[1 1 1] Update Visuals
  ;;[1 1 1] CEs are done for child blueprints
  (bind _asmaterial_interface (Utilities|Casting|CastToMaterialInterface LoadedObject))
  (Rendering|Material|SetMaterial (Variables|Default|GetLightCone) 0 _asmaterial_interface)
  (Variables|Default|SetCurrentLightConeMaterial _asmaterial_interface))

(event Custom|MeshMaterialLoaded
  (params (Object :type Object%20Reference))
  ;[1 1 1] Update Visuals
  ;;[1 1 1] CEs are done for child blueprints
  (bind _asmaterial_interface (Utilities|Casting|CastToMaterialInterface Object))
  (Rendering|Material|SetMaterial (Variables|Debug|GetStaticMeshComponent) 0 _asmaterial_interface)
  (Variables|Default|SetCurrentMeshMaterial _asmaterial_interface))

(fn GetOverrideMeshMaterial
  (params (MeshComp :type Static%20Mesh%20Component%20Object%20Reference) (MatIdx :type Integer))
  (Utilities|FlowControl|Sequence
    (:then_0
      (if (Utilities|Equal(Object) (Variables|GetMeshComp) (Variables|Default|GetLightCone))
        (bind _currentlightconematerial (Variables|Default|GetCurrentLightConeMaterial)
          (:then
            (return _currentlightconematerial))
          (:else))))
    (:then_1
      (if (Utilities|Equal(Object) (Variables|GetMeshComp) (Variables|Default|GetFlare))
        (bind _currentflarematerial (Variables|Default|GetCurrentFlareMaterial)
          (:then
            (return _currentflarematerial))
          (:else))))
    (:then_2
      (if (Utilities|Equal(Object) (Variables|Debug|GetStaticMeshComponent) (Variables|GetMeshComp))
        (bind _currentmeshmaterial (Variables|Default|GetCurrentMeshMaterial)
          (:then
            (return _currentmeshmaterial))
          (:else))))))

(fn SetFlareVisibility
  (params (Component :type Scene%20Component%20Object%20Reference) (Index :type Integer))
  (Development|SetHiddeninGame Component (== Index 0) false))

(fn ConstructionScript
  (bind _flare (Variables|Default|GetFlare))
  (Networking|CanExecuteCosmeticEvents
    (:True
      (Variables|/Script/Engine.SceneComponent|SetbAbsoluteRotation true _flare)
      (Transformation|SetWorldRotation _flare "0, 0,90.000000" false false))))
