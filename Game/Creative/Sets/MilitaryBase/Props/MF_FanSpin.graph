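# Material function that spins or oscillates fan-style props through World Position Offset, correcting normals, behind the UseFanRotation switch.
# An instance-space axis and pivot bias drive RotateAboutAxis and FixRotateAboutAxisNormals.
# A static switch picks steady spin or clamped-sine oscillation with pauses and position-based randomness.
# Spin can loop over one fan section to reduce Nanite displacement, using PreviousFrameSwitch to fix velocities at the wrap.
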MaterialFunction /Game/Creative/Sets/MilitaryBase/Props/MF_FanSpin
graph {
 properties {description="This function is used to give assets rotation on the X, Y, and Z-axis. Normalized rotation along with fixed normals. "}
 nodes [
  n1 = FunctionInput() {inputName=Material Attributes, inputType=FunctionInput_MaterialAttributes}
  n2 = GetMaterialAttributes(None=@n1) {attributeGetTypes [#1]}
  n3 = MaterialFunctionCall(Rotation Axis=NamedRerouteUsage(), Rotation Angle=NamedRerouteUsage(), World Space Vertex Normals=Transform(None=@n2.Normal)) {materialFunction=&FixRotateAboutAxisNormals}
  Add(A=@n3.New Tangent Space Vertex Normal, B=Multiply(A=@n2.Normal, B=Constant3Vector() {constant=<rgba 1 1 0 0>}))
  n4 = RotateAboutAxis(
   NormalizedRotationAxis=NamedRerouteUsage()
   RotationAngle=NamedRerouteUsage()
   PivotPoint=TransformPosition(None=VectorParameter() {group=FanRotation, parameterName=PivotPointBias}.RGB) {transformSourceType=TRANSFORMPOSSOURCE_Instance, transformType=TRANSFORMPOSSOURCE_World}
   Position=WorldPosition().XYZ
  )
  FunctionOutput(None=StaticSwitchParameter(
    True=SetMaterialAttributes!(MaterialAttributes=@n2.MaterialAttributes, WorldPositionOffset=@n4, Normal=@n3.New Tangent Space Vertex Normal) {attributeSetTypes [#2, #1]}
    False=@n1
   ) {group=FanRotation, parameterName=UseFanRotation})
  n5 = ScalarParameter() {defaultValue=0.1, group=FanRotation, parameterName=FanRotation_OscillatingAngle}
  n6 = ScalarParameter() {defaultValue=1.1, group=FanRotation, parameterName=FanRotation_OscillatingPause}
  n7 = Time() {period=5}
  n8 = ScalarParameter() {defaultValue=1, group=FanRotation, parameterName=FanRotation_Randomess}
  n9 = MaterialFunctionCall(Float3=TransformPosition(None=Constant3Vector()) {transformSourceType=TRANSFORMPOSSOURCE_Instance, transformType=TRANSFORMPOSSOURCE_World}) {materialFunction=&BreakOutFloat3Components}
  n10 = Add(A=@n9.R, B=@n9.R)
  n11 = Frac(None=Divide(A=Add(A=@n10, B=@n9.B)) {constB=10})
  n12 = Add(A=Multiply(A=@n8, B=Multiply(A=@n11) {constB=33.3}), B=ScalarParameter() {group=FanRotation, parameterName=FanRotation_ManualTimeOffset})
  n13 = Multiply(A=Multiply(A=@n11) {constB=0.01}, B=@n8)
  n14 = ScalarParameter() {defaultValue=0.2, group=FanRotation, parameterName=FanRotationSpeed}
  n15 = Sine(None=Multiply(A=Add(A=@n7, B=@n12), B=Add(A=@n13, B=@n14)))
  n16 = Abs(None=@n15)
  n17 = Multiply(A=Saturate(None=Multiply(A=@n6, B=@n16)), B=Sign(None=@n15))
  n18 = Frac(None=Multiply(A=Time() {period=5}, B=@n14))
  n19 = Multiply(A=Time(), B=@n14)
  n20 = Constant() {r=360}
  n21 = ScalarParameter() {
   defaultValue=30
   desc="If FanSectionLooping > 0 then this is used to split the 360 rotation"
   group=FanRotation
   parameterName=FanSectionCount
   sortPriority=34
  }
  n22 = Divide(A=Divide(A=@n20, B=@n21), B=@n20)
  n23 = Multiply(A=DeltaTime(), B=@n14)
  n24 = Add(A=@n19, B=@n23)
  n25 = PreviousFrameSwitch(Current Frame=Modulo!(A=@n19, B=@n22), Previous Frame=Subtract(A=Modulo(A=@n24, B=@n22), B=@n23)) {desc=Fixes velocity issues when we jump back to start of loop}
  n26 = ScalarParameter() {
   desc="If > 0 then we loop the rotation after a certain amount based on section count.  Split the total 360 rotation in FanSectionCount sections and loop the 1 section over an over.\r\nMainly used to optimize nanite by requiring less max displacement value for objects that spin fast enough to not notice issues with velocities or repetition from using this like fast moving fans."
   group=FanRotation
   parameterName=FanSectionLooping
   sortPriority=33
  }
  n27 = Ceil(None=@n26)
  NamedRerouteDeclaration(None=StaticSwitchParameter(True=Multiply(A=@n5, B=@n17), False=LinearInterpolate(A=@n18, B=@n25, Alpha=@n27)) {group=FanRotation, parameterName=UseFanRotation_Oscillating}) {name=l_FanRotationAmount, nodeColor=<rgba 0 1 0.2031 1>}
  n28 = VectorParameter() {group=FanRotation, parameterName=NormalizedRotationAxis, defaultValue=<rgba 1 0 0 0>}
  NamedRerouteDeclaration(None=Normalize(VectorInput=Transform(None=@n28.RGB) {transformSourceType=TRANSFORMSOURCE_Instance})) {name=l_FanRotationAxis, nodeColor=<rgba 0.5469 0 1 1>}
  Reroute(None=Reroute())
 ]
}
